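# Material function for the small-orb accent look on the NebulaBurdenCache backpack, output both mask-blended over an input material and unmasked.
# Two color parameters blend by a Fresnel mask that peaks at both the rim and the view-facing center.
# Emissive boosts the second color, dims under the utility function's glass thickness mask, then scales by a power parameter.
# The same utility function supplies glass color for base color and gem roughness.
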MaterialFunction /BRCosmetics/Accessories/FORT_Backpacks/Backpack_M_MED_NebulaBurdenCache/MaterialFunctions/MF_NebulaBurdenCache_Accent
graph {nodes [
  n1 = VectorParameter() {group=_Colors, parameterName=Small Orbs Color A, sortPriority=13, defaultValue=<rgba 0.0781 0.0094 0.0049 1>}
  n2 = Saturate(None=Fresnel() {baseReflectFraction=0, exponent=1})
  n3 = Max(A=Fresnel() {baseReflectFraction=0}, B=Power(Base=OneMinus(None=@n2)) {constExponent=12})
  n4 = VectorParameter() {group=_Colors, parameterName=Small Orbs Color B, sortPriority=20, defaultValue=<rgba 0.8021 0.1849 0.0249 1>}
  n5 = LinearInterpolate(A=@n1.RGB, B=@n4.RGB, Alpha=@n3)
  n6 = MaterialFunctionCall!(WarpMask) {materialFunction=&/MF_NebulaBurdenCache_Utilities}
  n7 = LinearInterpolate(A=@n5, B=Multiply(A=@n4.RGB, B=ScalarParameter() {defaultValue=1, group=Small Orb, parameterName=Small Orb Enter Emissive Boost}), Alpha=@n3)
  n8 = Multiply(A=@n7) {constB=0.25}
  n9 = MakeMaterialAttributes!(
   BaseColor=LinearInterpolate(A=@n5, B=@n6.Glass Color, Alpha=@n6.Glass Color)
   Metallic
   Specular
   Roughness=@n6.Gem Roughness
   Anisotropy
   EmissiveColor=Multiply(
    A=LinearInterpolate(A=@n7, B=@n8, Alpha=@n6.Glass Thickness Mask) {constB=0}
    B=ScalarParameter() {defaultValue=1, group=Small Orb, parameterName=Small Orb Emissive Power}
   )
   Opacity
   OpacityMask
   Normal
   Tangent
   WorldPositionOffset
   SubsurfaceColor
   ClearCoat
   ClearCoatRoughness
   AmbientOcclusion
   Refraction
   CustomizedUVs_0
   CustomizedUVs_1
   CustomizedUVs_2
   CustomizedUVs_3
   CustomizedUVs_4
   CustomizedUVs_5
   CustomizedUVs_6
   CustomizedUVs_7
   PixelDepthOffset
   ShadingModel
   Displacement
  )
  FunctionOutput(None=BlendMaterialAttributes(
    A=FunctionInput() {inputName=Material, inputType=FunctionInput_MaterialAttributes}
    B=@n9
    Alpha=FunctionInput() {inputName=Mask, inputType=FunctionInput_Scalar, sortPriority=1}
   )) {outputName=Material}
  FunctionOutput(None=@n9) {outputName=Material Non Masked, sortPriority=1}
 ]}
