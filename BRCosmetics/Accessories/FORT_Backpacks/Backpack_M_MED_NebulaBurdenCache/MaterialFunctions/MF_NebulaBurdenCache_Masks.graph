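# Material function for the NebulaBurdenCache backpack that outputs masks separating its center gem, side gems and accent gems.
# Three sphere masks use vector parameters for local offset and radius, offsetting from the utility function's PreSkinnedWorldPos and scaling radius by object X scale.
# The two side masks merge with Max, while the Left/Right output carries only Side Orb 1.
# Accent Gems is the inverse of all three orbs combined.
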MaterialFunction /BRCosmetics/Accessories/FORT_Backpacks/Backpack_M_MED_NebulaBurdenCache/MaterialFunctions/MF_NebulaBurdenCache_Masks
graph {nodes [
  n1 = VectorParameter() {group=Center Orb | General, parameterName=Orb Center Mask, channelNames {a=Radius, b=Z, g=Y, r=X}, defaultValue=<rgba 0 0 0 5>}
  n2 = MaterialFunctionCall!(WarpMask) {materialFunction=&/MF_NebulaBurdenCache_Utilities}
  n3 = SphereMask(
   A=WorldPosition().XYZ
   B=Add(A=@n2.PreSkinnedWorldPos, B=Transform(None=@n1) {transformSourceType=TRANSFORMSOURCE_Local}) {constB=0}
   Radius=Multiply(
    A=MaterialFunctionCall() {materialFunction=&ObjectScale}.Scale X
    B=@n1.Radius
   )
  ) {desc=Center Orb Mask}
  FunctionOutput(None=@n3) {outputName=Center Gem, sortPriority=1}
  n4 = VectorParameter() {group=Side Orb | Orb Mask, parameterName=Side Orb 2 Mask, channelNames {a=Radius, b=Z, g=Y, r=X}, defaultValue=<rgba 0 0 0 5>}
  n5 = Transform(None=@n4) {transformSourceType=TRANSFORMSOURCE_Local}
  n6 = MaterialFunctionCall() {materialFunction=&ObjectScale}
  n7 = VectorParameter() {group=Side Orb | Orb Mask, parameterName=Side Orb 1 Mask, channelNames {a=Radius, b=Z, g=Y, r=X}, defaultValue=<rgba 0 0 0 5>}
  n8 = SphereMask(
   A=WorldPosition().XYZ
   B=Add(A=@n2.PreSkinnedWorldPos, B=Transform(None=@n7) {transformSourceType=TRANSFORMSOURCE_Local})
   Radius=Multiply(A=@n7.Radius, B=@n6.Scale X)
  )
  n9 = Max(A=SphereMask(A=WorldPosition().XYZ, B=Add(A=@n2.PreSkinnedWorldPos, B=@n5), Radius=Multiply(A=@n4.Radius, B=@n6.Scale X)), B=@n8)
  FunctionOutput(None=OneMinus(None=Max(A=@n3, B=@n9) {desc=Center Orb + Side Orbs})) {outputName=Accent Gems, sortPriority=4}
  FunctionOutput(None=@n9) {outputName=Side Gems Combined, sortPriority=2}
  FunctionOutput(None=@n8) {outputName=Side Gems Left/Right Mask, sortPriority=3}
 ]}
